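# Sound cue for lightning strike sounds on the Hightower Squash glider cosmetic.
# A Random node picks one of four equally weighted strike waves per play.
# Playback is very quiet, uses the shared player glider open/close attenuation and routes to the glider landing sound class.
# A concurrency override caps simultaneous voices at a few and blocks new plays beyond the limit.
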SoundCue /BRCosmetics/Sounds/Gliders/Hightower_Squash/Hightower_Squash_Strikes_Cue
root=Random_0
nodes [
 {id=Random_0, children [WavePlayer_1,WavePlayer_2,WavePlayer_3,WavePlayer_4], properties {weights [1,1,1,1]}}
 {id=WavePlayer_1, properties {soundWaveAssetPtr=&/GIlder_Hightower_Squash_LightningStrike_v1_04}}
 {id=WavePlayer_2, properties {soundWaveAssetPtr=&/GIlder_Hightower_Squash_LightningStrike_v1_01}}
 {id=WavePlayer_3, properties {soundWaveAssetPtr=&/GIlder_Hightower_Squash_LightningStrike_v1_02}}
 {id=WavePlayer_4, properties {soundWaveAssetPtr=&/GIlder_Hightower_Squash_LightningStrike_v1_03}}
]
properties {
 attenuationSettings=/Game/Sounds/Attenuation/Fort_Player_Glider_OpenClose_Athena_Att
 bOverrideConcurrency=true
 maxDistance=10000
 soundClassObject=&Fort_Gliders_Land
 volumeMultiplier=0.04
 concurrencyOverrides {bVolumeScaleCanRelease=true, maxCount=4, resolutionRule=PreventNew, voiceStealReleaseTime=0.35, volumeScaleAttackTime=0.3}
 concurrencySet [null]
}
