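# Niagara effect for the HighTower Squash glider, layering electric arcs, lightning, energy flow, glow, orbiting debris, clouds and motion-gated ribbon trails.
# Most emitters follow bones or sockets on User.SkeletalMesh, with spawn groups splitting arcs between paired arms and hands.
# System-level random burst, group and loop values synchronize lightning meshes and sparks, which play strike audio.
# Particle lights accent glow and lightning, and clouds stop spawning in front-end preview.
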NiagaraSystem /BRCosmetics/Gadgets/Assets/VinderTech_GliderChute/Glider_HighTower_Squash/FX/NS_Glider_HighTower_Squash
dependencies {
 usedDataInterfaces [
  &NiagaraDataInterfaceVectorCurve
  &NiagaraDataInterfaceCurve
  &NiagaraDataInterfaceColorCurve
  &NiagaraDataInterfaceAudioPlayer
 ]
 usedDynamicInputs [&Lerp_Float, &MakeIntFromFloat]
 usedModules [
  &SystemState
  &self:SystemSpawnScript.SystemScriptSource.SystemScriptGraph.NiagaraNodeAssignment_1.SetVariables_#1
  &self:SystemSpawnScript.SystemScriptSource.SystemScriptGraph.NiagaraNodeAssignment_3.SetVariables_#2
  /Game/Effects/Niagara/Modules/Math/FN_Integrator/FN_Integrator_v2
  &self:SystemSpawnScript.SystemScriptSource.SystemScriptGraph.NiagaraNodeAssignment_0.SetVariables_#3
  &self:SystemSpawnScript.SystemScriptSource.SystemScriptGraph.NiagaraNodeAssignment_2.SetVariables_#4
  &self:SystemSpawnScript.SystemScriptSource.SystemScriptGraph.NiagaraNodeAssignment_4.SetVariables_#5
 ]
}
emitters {
 AirGlow {
  emitterUpdateScript [
   &EmitterState()
   &SpawnRate(SpawnRate=10, Use Spawn Probability=true)
  ]
  particleSpawnScript [
   /Niagara/Modules/Spawn/Initialization/InitializeParticle(
    Color=&UniformRangedLinearColor(Minimum=<rgba 0.1209 0.1466 0.15 0.35>, Maximum=<rgba 0.2708 0.3299 0.349 0.25>)
    Lifetime=0.3
    Module.Write Color=true
    Module.Write Lifetime=true
    Module.Write Mass=false
    Module.Write Material Random=true
    Module.Write Position=false
    Module.Write Scale=false
    Module.Write SpriteRotation=false
    Module.Write SpriteSize=true
    Position=&SimulationPosition()
    Sprite Size=&UniformRangedVector2D(Minimum=<100 100>, Maximum=<120 120>)
   )
   &DynamicMaterialParameters(
    Index 0 Param 1=0.5
    Index 0 Param 2=30
    Module.Param0WriteEnabled=true
    Module.Param1WriteEnabled=true
    Module.Param2WriteEnabled=true
    Module.Param3WriteEnabled=true
   )
  ]
  particleUpdateScript [
   &ParticleState()
   &SkeletalMeshLocation(
    Apply To Particle Position=false
    Bone / Socket Index=1
    Bone Sampling Mode=::NewEnumerator4
    Inherit Velocity=false
    Interpolate Mesh Sampling=false
    MaskSpawnGroup=false
    Sampled Position Offset Coordinate Space:ENiagaraMeshTransforms=::NewEnumerator3
    Skeletal Mesh=$User.SkeletalMesh
   )
   /SetVariables() {isSetParameters=true}
   &ScaleSpriteSize(Scale Factor=&Vector2DFromFloat(Value=&FloatFromCurve(FloatCurve=DataInterface() {propertyValues {Curve {keys [<at 0.25 1>,<at 1 0.75>]}}})))
   &ScaleColor(
    Scale Alpha=&FloatFromCurve(FloatCurve=DataInterface() {propertyValues {Curve {keys [<at 0.1 0>,<at 0.2 1>,<at 0.75 1>,<at 0.9 0>]}}})
    Scale RGB=&MakeVectorFromLinearColorRGB(Color=&ColorFromCurve(ColorCurve=DataInterface() {propertyValues {
        AlphaCurve {keys []}
        BlueCurve {keys [<at 0.25 1 RCIM_Cubic>,<at 1 0.0521 RCIM_Cubic>]}
        GreenCurve {keys [<at 0.25 1 RCIM_Cubic>,<at 1 0.0521 RCIM_Cubic>]}
        RedCurve {keys [<at 0.25 1 RCIM_Cubic>,<at 1 0.0521 RCIM_Cubic>]}
       }}))
    ScaleA=true
    ScaleRGB=true
   )
   &FN_Light_v3(
    Color=<rgba 6.0192 7.5173 8 1>
    Create Color Attribute=true
    Create Exponent Attribute=true
    Create Position Attribute=true
    Create Radius Attribute=true
    Enable Particle Light Enable=true
    Exponent=1
    Light Enabled=HlslExpression() {hlslExpression="Modulo (Particles.UniqueID, 10) == 1 ? 1.0: 0"}
    Position:Vector3f=&Add_Vector(
     A=$Output.SkeletalMeshLocation.SampledPosition
     B=&RotateVectorByQuaternion(Vector To Rotate=<-32 10 7>, Quaternion=$Output.SkeletalMeshLocation.SampledRotation)
    )
    Radius=60
    Scale Color:NiagaraFloat=&FloatFromCurve(FloatCurve=DataInterface() {propertyValues {Curve {keys [<at 0 1>,<at 0.25 10>,<at 0.34 3.8638>,<at 0.75 5.9718>,<at 1 1>]}}})
   )
  ]
  rendererClasses [&NiagaraSpriteRendererProperties, &NiagaraLightRendererProperties]
  renderers [
   {rendererClass=&NiagaraSpriteRendererProperties}
   {rendererClass=&NiagaraLightRendererProperties, rendererIndex=1}
  ]
 }
 Clouds {
  emitterUpdateScript [
   &EmitterState()
   &SpawnRate(
    SpawnRate=&MakeCustomFloatFromBool(Bool=$User.bIsFrontEndPreview, True Float=0, False Float=8)
    Use Spawn Probability=false
   )
  ]
  particleSpawnScript [
   &InitializeParticle(
    Color=&MakeLinearColor(
     R:NiagaraFloat=1
     G:NiagaraFloat=1
     B:NiagaraFloat=1
     A:NiagaraFloat=&UniformRangedFloat(Minimum=0.05, Maximum=0.1)
    )
    Lifetime=&UniformRangedFloat(Minimum=3, Maximum=6)
    Mass=&UniformRangedFloat(Minimum=0.6, Maximum=1.2)
    Mesh Scale=<50 50 50>
    Module.Write Color=true
    Module.Write Lifetime=true
    Module.Write Mass=true
    Module.Write Material Random=true
    Module.Write Position=false
    Module.Write Scale=true
    Module.Write SpriteRotation=true
    Module.Write SpriteSize=true
    Position=&SimulationPosition()
    Sprite Rotation=&NormalizedAngleToDegrees(NormalizedAngle=&UniformRangedFloat(Minimum=0, Maximum=1))
    Sprite Size=&Vector2DFromFloat(Value=&UniformRangedFloat(Minimum=1800, Maximum=2300))
   )
   /Niagara/Modules/Spawn/Location/SphereLocation(
    MaskSpawnGroup=false
    Module.Override Randomness=false
    Sphere Origin:NiagaraPosition=&Add_Vector(
     A=Linked() {linkedVariable {name=Engine.Owner.Position, type=NiagaraPosition}}
     B=&RotateVectorByQuaternion(Vector To Rotate=<0 750 0>, Quaternion=$Engine.Owner.Rotation)
    )
    Sphere Radius=2000
    Surface Only=false
   )
   /SetVariables(Particles.RandomRotationRate:NiagaraFloat=&UniformRangedFloat(Minimum=5, Maximum=25)) {isSetParameters=true}
   /SetVariables(Particles.SubImageIndex=&UniformRangedFloat(Minimum=0, Maximum=63)) {isSetParameters=true}
  ]
  particleUpdateScript [
   &ParticleState()
   &DynamicMaterialParameters(
    Index 0 Param 1=&FloatFromCurve(FloatCurve=DataInterface() {propertyValues {Curve {keys [<at 0 1>,<at 0.25 1>,<at 1 -1>]}}})
    Index 0 Param 4=0.2
    Module.Param0WriteEnabled=true
    Module.Param1WriteEnabled=true
    Module.Param2WriteEnabled=true
    Module.Param3WriteEnabled=true
   )
   &ScaleColor(
    Scale Alpha=&FloatFromCurve(FloatCurve=DataInterface() {propertyValues {Curve {keys [
         <at -0.0046 0 RCIM_Cubic>
         {arriveTangent=2.4716, leaveTangent=2.4716, time=0.2, value=1}
         {arriveTangent=-1.2458, interpMode=RCIM_Cubic, leaveTangent=-1.2458, time=0.25, value=1}
         <at 1.0011 0.002 RCIM_Cubic>
        ]}}})
    ScaleA=true
    ScaleRGB=true
   )
   &SpriteRotationRate(Rotation Rate=$Particles.RandomRotationRate)
  ]
  rendererClasses [&NiagaraSpriteRendererProperties]
  renderers [{rendererClass=&NiagaraSpriteRendererProperties}]
 }
 Debris {
  emitterUpdateScript [
   &EmitterState()
   &SpawnRate(SpawnRate=&UniformRangedFloat(Minimum=2, Maximum=6), Use Spawn Probability=true)
  ]
  particleSpawnScript [
   /Game/Effects/Niagara/Modules/Orbit/FN_Orbit_Init_Variables_v2()
   &InitializeParticle(
    Lifetime=&UniformRangedFloat(Minimum=1, Maximum=1.5)
    Mesh Scale=&UniformRangedVector(Minimum=<0.1 0.1 0.1>, Maximum=<0.35 0.35 0.35>)
    Module.Write Color=true
    Module.Write Lifetime=true
    Module.Write Mass=false
    Module.Write Material Random=true
    Module.Write Position=false
    Module.Write Scale=true
    Module.Write SpriteRotation=false
    Module.Write SpriteSize=false
    Position=&SimulationPosition()
   )
  ]
  particleUpdateScript [
   &ParticleState()
   &SkeletalMeshLocation(
    Apply To Particle Position=false
    Bone / Socket Index=1
    Bone Sampling Mode=::NewEnumerator4
    Inherit Velocity=false
    Interpolate Mesh Sampling=false
    MaskSpawnGroup=false
    Skeletal Mesh=$User.SkeletalMesh
   )
   &ScaleMeshSize(Scale Factor=&VectorFromFloat(Value=&FloatFromCurve(FloatCurve=DataInterface() {propertyValues {Curve {keys [
          <at 0 0 RCIM_Cubic>
          <at 0.15 1 RCIM_Cubic>
          {arriveTangent=-3.9409, interpMode=RCIM_Cubic, leaveTangent=-3.9409, tangentMode=RCTM_User, time=1, value=0}
         ]}}})))
   &ScaleColor(
    Scale Alpha=&FloatFromCurve(FloatCurve=DataInterface() {propertyValues {Curve {keys [
         <at 0 0 RCIM_Cubic>
         <at 0.3 1 RCIM_Cubic>
         {arriveTangent=-4.5166, interpMode=RCIM_Cubic, leaveTangent=-4.5166, tangentMode=RCTM_User, time=1, value=0}
        ]}}})
    ScaleA=true
    ScaleRGB=false
   )
   /SetVariables() {isSetParameters=true}
   /Game/Effects/Niagara/Modules/Orbit/FN_Orbit_v4(
    Height min/max:Vector2f=<10 30>
    Radius min/max:Vector2f=<30 80>
    Rotation Offset min/max:Vector2f=<0 0.25>
    Rotation Rate min/max:Vector2f=<0.5 2>
   )
  ]
  rendererClasses [&NiagaraMeshRendererProperties]
  renderers [{rendererClass=&NiagaraMeshRendererProperties, rendererIndex=0}]
 }
 ElectricMeshArcsArms {
  emitterUpdateScript [
   &EmitterState(
    Life Cycle Mode=::NewEnumerator1
    Loop Duration=&UniformRangedFloat(Minimum=0.175, Maximum=0.225)
    Recalculate Duration Each Loop=true
    UseLoopDelay=false
   )
   &SpawnBurst_Instantaneous(Spawn Count=&UniformRangedInt(Minimum=1, Maximum=3), Use Spawn Probability=false)
  ]
  particleSpawnScript [
   &FN_SetSpawnGroup_v2(Number of Spawn Groups=2)
   &InitializeParticle(
    Color=&UniformRangedLinearColor(Minimum=<rgba 1 1 1 1>, Maximum=<rgba 3 3 3 1>)
    Lifetime=&UniformRangedFloat(Minimum=0.1, Maximum=0.15)
    Mesh Scale=&MakeVector(X=&MaskFloatBySpawnGroup(Masked Float=1, Passthrough Float=-1), Y=1, Z=1)
    Module.Write Color=true
    Module.Write Lifetime=true
    Module.Write Mass=true
    Module.Write Material Random=true
    Module.Write Position=false
    Module.Write Scale=true
    Module.Write SpriteRotation=true
    Module.Write SpriteSize=true
    Sprite Size=<1 1>
   )
   &DynamicMaterialParameters(
    Index 0 Param 1=&UniformRangedFloat(Minimum=0.75, Maximum=0.95)
    Index 0 Param 2=&UniformRangedFloat(Minimum=1, Maximum=2)
    Index 0 Param 3=&UniformRangedFloat(Minimum=0, Maximum=1)
    Index 1 Param 1=&UniformRangedFloat(Minimum=0, Maximum=1)
    Index 1 Param 2=&UniformRangedFloat(Minimum=0, Maximum=20)
    Module.Param0WriteEnabled=true
    Module.Param0WriteEnabled1=true
    Module.Param1WriteEnabled=true
    Module.Param1WriteEnabled1=true
    Module.Param2WriteEnabled=true
    Module.Param2WriteEnabled1=false
    Module.Param3WriteEnabled=false
    Module.Param3WriteEnabled1=false
    Write Parameter Index 1=true
   )
   /Niagara/Modules/Spawn/Orientation/V2/InitialMeshOrientation(
    Rotation=&UniformRangedVector(Minimum=<-1 0 0>, Maximum=<1 0 0>)
    Rotation Coordinate Space:ENiagaraMeshTransforms=::NewEnumerator0
    Use Orientation Vector=true
    Use Rotation Vector=true
   )
  ]
  particleUpdateScript [
   &ParticleState()
   &SkeletalMeshLocation(
    Apply Sampled Orientation to Mesh Particles=true
    Bone / Socket Index=&Add_Integer(A=$Particles.SpawnGroup, B=15)
    Bone Sampling Mode=::NewEnumerator4
    Inherit Velocity=false
    Interpolate Mesh Sampling=false
    MaskSpawnGroup=false
    Skeletal Mesh=$User.SkeletalMesh
   )
   &ScaleColor(
    Scale Alpha=&FloatFromCurve(FloatCurve=DataInterface() {propertyValues {Curve {keys [<at 0.75 1 RCIM_Cubic>,<at 1 0>]}}})
    Scale RGB=&MakeVectorFromLinearColorRGB(Color=&ColorFromCurve(ColorCurve=DataInterface() {propertyValues {
        AlphaCurve {keys []}
        BlueCurve {keys [<at 0 1 RCIM_Cubic>,<at 1 3 RCIM_Cubic>]}
        GreenCurve {keys [<at 0 0.9955 RCIM_Cubic>,<at 1 2.8618 RCIM_Cubic>]}
        RedCurve {keys [<at 0 0.9844 RCIM_Cubic>,<at 1 2.6719 RCIM_Cubic>]}
       }}))
    ScaleA=true
    ScaleRGB=true
   )
   &ScaleMeshSize(Scale Factor=&MakeVector(
     X=&FloatFromCurve(FloatCurve=DataInterface() {propertyValues {Curve {keys [<at 0 0.5 RCIM_Cubic>, {arriveTangent=1, leaveTangent=1, time=0.25, value=1}, <at 1 1>]}}})
     Y=1
     Z=&FloatFromCurve(FloatCurve=DataInterface() {propertyValues {Curve {keys [<at 0.8 1 RCIM_Cubic>, {arriveTangent=-11.4579, interpMode=RCIM_Cubic, leaveTangent=-11.4579, tangentMode=RCTM_User, time=1, value=0}]}}})
    ))
   /SetVariables(Particles.MeshOrientation=&MultiplyQuaternion(Quaternion A=$Output.SkeletalMeshLocation.SampledRotation, Quaternion B=$Particles.Initial.MeshOrientation)) {isSetParameters=true}
  ]
  rendererClasses [&NiagaraMeshRendererProperties]
  renderers [{rendererClass=&NiagaraMeshRendererProperties, rendererIndex=0}]
 }
 Electric_Arcs_Hands {
  emitterUpdateScript [
   &EmitterState(
    Inactive Response=::NewEnumerator1
    Life Cycle Mode=::NewEnumerator1
    Loop Duration=&UniformRangedFloat(Minimum=0.05, Maximum=0.3)
    Recalculate Duration Each Loop=true
    UseLoopDelay=false
   )
   /SetVariables(Emitter.SpawnBurstRange:NiagaraInt32=&UniformRangedInt(Minimum=2, Maximum=4)) {isSetParameters=true}
   &SpawnBurst_Instantaneous(Spawn Count=$Emitter.SpawnBurstRange, Use Spawn Probability=true)
  ]
  particleSpawnScript [
   &FN_SetSpawnGroup_v2(Number of Spawn Groups=2)
   &InitializeParticle(
    Color=<rgba 12.1042 13.2767 14 0.9>
    Lifetime=&UniformRangedFloat(Minimum=0.15, Maximum=0.2)
    Module.Write Color=true
    Module.Write Lifetime=true
    Module.Write Mass=true
    Module.Write Material Random=true
    Module.Write Position=true
    Module.Write Scale=false
    Module.Write SpriteRotation=false
    Module.Write SpriteSize=true
    Sprite Size=&UniformRangedVector2D(Minimum=<15 30>, Maximum=<20 45>)
   )
   &DynamicMaterialParameters(
    Index 0 Param 1=&UniformRangedFloat(Minimum=2.5, Maximum=3.5)
    Index 0 Param 3=&UniformRangedFloat(Minimum=0.2, Maximum=0.5)
    Module.Param0WriteEnabled=true
    Module.Param1WriteEnabled=false
    Module.Param2WriteEnabled=true
    Module.Param3WriteEnabled=true
   )
   &CameraOffset(Camera Offset Amount=&UniformRangedFloat(Minimum=-10, Maximum=10))
   /SetVariables(Particles.LightningUVScale:Vector2f=&MakeVector2D(X=&UniformAOrBFloat(A=-1), Y=&UniformAOrBFloat(A=-1))) {isSetParameters=true}
   /SetVariables(
    Particles.ShortLightningArcsColor:LinearColor=<rgba 0.9453 1.041 1.1 0.1>
    Particles.ShortLightningArcsSize:Vector2f=&Vector2DFromFloat(Value=&UniformRangedFloat(Minimum=24, Maximum=30))
   ) {isSetParameters=true}
   &InitialMeshOrientation(
    Rotation=&UniformRangedVector(Minimum=<-1 0 0>, Maximum=<1 0 0>)
    Rotation Coordinate Space:ENiagaraMeshTransforms=::NewEnumerator0
    Use Orientation Vector=true
    Use Rotation Vector=true
   )
  ]
  particleUpdateScript [
   &ParticleState()
   &SkeletalMeshLocation(
    Apply Sampled Orientation to Mesh Particles=true
    Bone / Socket Index=&Add_Integer(A=$Particles.SpawnGroup, B=17)
    Bone Sampling Mode=::NewEnumerator4
    Inherit Velocity=false
    Interpolate Mesh Sampling=false
    MaskSpawnGroup=false
    Skeletal Mesh=$User.SkeletalMesh
   ) {module=SkeletalMeshLocation001}
   /SetVariables() {isSetParameters=true}
   /Game/Effects/Niagara/Modules/Mesh/FN_MeshRenderer_v3/FN_MeshRenderer_v2(
    Color:LinearColor=&MultiplyLinearColor(
     Linear Color=<rgba 0.6406 0.7354 0.75 0.1>
     Scale Factor=&MakeFloat4FromVectorAndFloat(
      "Vector (RGB)"=&MakeVectorFromLinearColorRGB(Color=&ColorFromCurve(ColorCurve=DataInterface() {propertyValues {
          AlphaCurve {keys []}
          BlueCurve {keys [<at 0 1.125 RCIM_Cubic>,<at 0.2 2 RCIM_Cubic>,<at 0.5 1.125 RCIM_Cubic>]}
          GreenCurve {keys [<at 0 1.0571 RCIM_Cubic>,<at 0.2 1.8793 RCIM_Cubic>,<at 0.5 1.0571 RCIM_Cubic>]}
          RedCurve {keys [<at 0 0.8464 RCIM_Cubic>,<at 0.2 1.5048 RCIM_Cubic>,<at 0.5 0.8464 RCIM_Cubic>]}
         }}))
      "Float (Alpha)"=&FloatFromCurve(FloatCurve=DataInterface() {propertyValues {Curve {keys [
           <at 0.1 1>
           {arriveTangent=-0.7039, interpMode=RCIM_Cubic, leaveTangent=-0.7039, time=0.3472, value=1}
           <at 0.5972 0.65 RCIM_Cubic>
           <at 0.8472 1>
           <at 1.0972 0.65 RCIM_Cubic>
           <at 1.3472 1 RCIM_Cubic>
           <at 1.5972 0.65>
           <at 1.8472 1 RCIM_Cubic>
          ]}}})
     )
    )
    Create COlor Attribute=true
    Create Orientation Attribute=true
    Create Position Attribute=true
    Create Scale Attribute=true
    Position Offset:Vector3f=&RotateVectorByQuaternion(
     Vector To Rotate=&MaskVectorBySpawnGroup(Masked Vector=<36 0 0>, Passthrough Vector=<-36 0 0>)
     Quaternion=$Output.SkeletalMeshLocation001.SampledRotation
    )
    Scale:Vector3f=<0.95 0.8 0.8>
   ) {module=FN_MeshRenderer_v2001}
   &ScaleSpriteSize(Scale Factor=&Vector2DFromFloat(Value=&FloatFromCurve(FloatCurve=DataInterface() {propertyValues {Curve {keys [{arriveTangent=1.9395, interpMode=RCIM_Cubic, leaveTangent=1.9395, tangentMode=RCTM_User, time=0, value=0.3}, <at 1 1 RCIM_Cubic>]}}})))
   &ScaleColor(
    Scale RGB=&MakeVectorFromLinearColorRGB(Color=&ColorFromCurve(ColorCurve=DataInterface() {propertyValues {
        AlphaCurve {keys []}
        BlueCurve {keys [<at 0 1.125 RCIM_Cubic>,<at 0.2 8 RCIM_Cubic>,<at 0.5 1.125 RCIM_Cubic>]}
        GreenCurve {keys [<at 0 1.0571 RCIM_Cubic>,<at 0.2 7.5173 RCIM_Cubic>,<at 0.5 1.0571 RCIM_Cubic>]}
        RedCurve {keys [<at 0 0.8464 RCIM_Cubic>,<at 0.2 6.0192 RCIM_Cubic>,<at 0.5 0.8464 RCIM_Cubic>]}
       }}))
    ScaleA=true
    ScaleRGB=true
   )
  ]
  rendererClasses [&NiagaraSpriteRendererProperties, &NiagaraMeshRendererProperties]
  renderers [
   {rendererClass=&NiagaraSpriteRendererProperties}
   {rendererClass=&NiagaraSpriteRendererProperties, rendererIndex=1}
   {rendererClass=&NiagaraMeshRendererProperties, rendererIndex=2}
  ]
 }
 EnergyFlow {
  emitterUpdateScript [
   &EmitterState()
   /Game/Effects/Niagara/Modules/Partition/FN_PartitionAttributes/FN_PartitionAttributes_v2(Particles Per Partition=5, Partitions=5) {enabled=false}
   &SpawnRate(SpawnRate=50, Use Spawn Probability=false)
   /Game/Effects/Niagara/Modules/Partition/FN_PartitionAttributes(Number of Bone Sets=5, Particles Per Partition=5, Partitions=5)
  ]
  particleSpawnScript [
   /Game/Effects/Niagara/Modules/Partition/FN_Particle_partition_v5/FN_Particle_Partition_v5()
   &InitializeParticle(
    Color=<rgba 2.5 2.8782 3 0.25>
    Lifetime=0.45
    Module.Write Color=true
    Module.Write Lifetime=true
    Module.Write Mass=false
    Module.Write Material Random=true
    Module.Write Position=false
    Module.Write Scale=false
    Module.Write SpriteRotation=true
    Module.Write SpriteSize=true
    Position=&SimulationPosition()
    Sprite Rotation=&NormalizedAngleToDegrees(NormalizedAngle=&UniformRangedFloat(Minimum=-1, Maximum=1))
    Sprite Size=HlslExpression() {hlslExpression="Particles.Partition.ID== 0 ? 50. : Particles.Partition.ID > 2 ? 32 :15 "}
   )
   /Game/Effects/Niagara/Modules/BoneFlow/FN_Initialize_BoneFlow_Attributes_v2(T:NiagaraFloat=&UniformRangedFloat(Minimum=0, Maximum=1))
   /Game/Effects/Niagara/Modules/Sprite/FN_Sprite_v2(
    Camera Offset=HlslExpression() {hlslExpression="Particles.Partition.ID== 0 ? 8. : Particles.Partition.ID > 2 ? 10 :6 "}
    Color:LinearColor=&MakeLinearColorFromVectorAndFloat(
     "Vector (RGB)"=&MakeVectorFromLinearColorRGB(Color=<rgba 1.1286 1.4095 1.5 1>)
     "Float (Alpha)"=&UniformRangedFloat(Minimum=0.25, Maximum=1)
    )
    Create Camera Offset Attribute=true
    Create Color Attribute=true
    "Create Normalized Age "=true
    Create Sprite Rotation Attribute=true
    Create SpriteSize Attribute=true
    Lifetime:NiagaraFloat=&MakeFloatFromInt(Int=&UniformRangedInt(Minimum=0, Maximum=3))
    SpriteSize:Vector2f=&Vector2DFromFloat(Value=&UniformRangedFloat(
      Minimum=HlslExpression() {hlslExpression="Particles.Partition.ID== 0 ? 24. : Particles.Partition.ID > 2 ? 16 :14"}
      Maximum=HlslExpression() {hlslExpression="Particles.Partition.ID== 0 ? 28. : Particles.Partition.ID > 2 ? 24 :18 "}
     ))
   ) {enabled=false}
   &DynamicMaterialParameters(
    Index 0 Param 2=&UniformRangedFloat(Minimum=0, Maximum=20)
    Index 0 Param 3=1
    Index 0 Param 4=1
    Index 1 Param 1=&UniformRangedFloat(Minimum=0, Maximum=0.1)
    Index 1 Param 2=&UniformRangedFloat(Minimum=1, Maximum=2)
    Module.Param0WriteEnabled=false
    Module.Param0WriteEnabled1=true
    Module.Param1WriteEnabled=true
    Module.Param1WriteEnabled1=true
    Module.Param2WriteEnabled=true
    Module.Param2WriteEnabled1=false
    Module.Param3WriteEnabled=true
    Module.Param3WriteEnabled1=false
    Write Parameter Index 1=true
   )
   &CameraOffset(Camera Offset Amount=7)
   /SetVariables(
    Particles.ArcSpriteCameraOffset=HlslExpression() {hlslExpression="Particles.Partition.ID== 0 ? 8. : Particles.Partition.ID > 2 ? 10 :6 "}
    Particles.ArcSpriteColor:LinearColor=&MakeLinearColorFromVectorAndFloat(
     "Vector (RGB)"=&MakeVectorFromLinearColorRGB(Color=<rgba 1.1286 1.4095 1.5 1>)
     "Float (Alpha)"=&UniformRangedFloat(Minimum=0.25, Maximum=1)
    )
    Particles.ArcSpriteSize:Vector2f=&Vector2DFromFloat(Value=&UniformRangedFloat(
      Minimum=HlslExpression() {hlslExpression="Particles.Partition.ID== 0 ? 24. : Particles.Partition.ID > 2 ? 16 :14"}
      Maximum=HlslExpression() {hlslExpression="Particles.Partition.ID== 0 ? 28. : Particles.Partition.ID > 2 ? 24 :18 "}
     ))
   ) {isSetParameters=true}
  ]
  particleUpdateScript [
   &ParticleState()
   &DynamicMaterialParameters(
    Index 0 Param 1=&FloatFromCurve(FloatCurve=DataInterface() {propertyValues {Curve {keys [<at 0 0.01>,<at 1 1.5>]}}})
    Index 1 Param 3=&FloatFromCurve(FloatCurve=DataInterface() {propertyValues {Curve {keys [<at 0 1 RCIM_Cubic>,<at 0.2 0 RCIM_Cubic>,<at 1 1>]}}})
    Module.Param0WriteEnabled=true
    Module.Param0WriteEnabled1=false
    Module.Param1WriteEnabled=false
    Module.Param1WriteEnabled1=false
    Module.Param2WriteEnabled=false
    Module.Param2WriteEnabled1=true
    Module.Param3WriteEnabled=false
    Module.Param3WriteEnabled1=false
    Write Parameter Index 1=true
   ) {module=DynamicMaterialParameters001}
   /Game/Effects/Niagara/Modules/BoneFlow/FN_BoneFlow_v3/FN_BoneFlow_v3(
    Angle:NiagaraFloat=$Particles.FN_Initialize_BoneFlow_Attributes_v2.Angle
    Radius:NiagaraFloat=$Particles.FN_Initialize_BoneFlow_Attributes_v2.Radius
    SkeletalMesh:NiagaraDataInterfaceSkeletalMesh=$User.SkeletalMesh
    T:NiagaraFloat=&Add_Float(A=$Particles.FN_Initialize_BoneFlow_Attributes_v2.T, B=&Multiply_Float(A=$Particles.NormalizedAge, B=0.1))
   )
   &ScaleColor(
    Scale Alpha=&FloatFromCurve(FloatCurve=DataInterface() {propertyValues {Curve {keys [<at 0 1>,<at 0.1 0>,<at 1 0>]}}})
    Scale RGB=&VectorFromFloat(Value=&FloatFromCurve(FloatCurve=DataInterface() {propertyValues {Curve {keys [<at 0 1>,<at 0.1 1.5>,<at 0.3094 1.2707>,<at 1 1>]}}}))
    ScaleA=true
    ScaleRGB=true
   )
   &ScaleSpriteSize(Scale Factor=&Vector2DFromFloat(Value=&FloatFromCurve(FloatCurve=DataInterface() {propertyValues {Curve {keys [<at 0 1>,<at 0.535 0.7711>,<at 0.9 1.1>]}}})))
   /SetVariables(Particles.ArcSpriteColor:LinearColor=&MultiplyLinearColor(
     Linear Color=$Particles.Initial.ArcSpriteColor
     Scale Factor=&MakeFloat4FromVectorAndFloat(
      "Vector (RGB)"=&MakeVectorFromLinearColorRGB(Color=&ColorFromCurve(ColorCurve=DataInterface() {propertyValues {
          AlphaCurve {keys []}
          BlueCurve {keys [<at 0 3 RCIM_Cubic>,<at 1 1 RCIM_Cubic>]}
          GreenCurve {keys [<at 0 2.7135 RCIM_Cubic>,<at 1 0.8177 RCIM_Cubic>]}
          RedCurve {keys [<at 0 2.5625 RCIM_Cubic>,<at 1 0.8188 RCIM_Cubic>]}
         }}))
      "Float (Alpha)"=&FloatFromCurve(FloatCurve=DataInterface() {propertyValues {Curve {keys [<at 0.65 1 RCIM_Cubic>, {arriveTangent=-4.5166, interpMode=RCIM_Cubic, leaveTangent=-4.5166, tangentMode=RCTM_User, time=1, value=0}]}}})
     )
    )) {isSetParameters=true}
  ]
  rendererClasses [&NiagaraSpriteRendererProperties]
  renderers [
   {rendererClass=&NiagaraSpriteRendererProperties}
   {rendererClass=&NiagaraSpriteRendererProperties, rendererIndex=1}
  ]
 }
 FN_AnimTrail_V4 {
  emitterUpdateScript [
   &EmitterState()
   /Game/Effects/Niagara/Modules/AnimTrail/FN_SpawnAnimTrail/FN_SpawnAnimTrail(
    Number of Anim Trails=2
    Spawn Mode:ENiagaraSkeletalSpawning=::NewEnumerator0
    SpawnRate:NiagaraFloat=&Multiply_Float(A=44, B=$System.MovingBool)
   )
  ]
  particleSpawnScript [
   /Game/Effects/Niagara/Modules/AnimTrail/FN_InitializeAnimTrail/FN_InitializeAnimTrailBoneIndecies(First Bone or Socket Index=4, Second Bone or Socket Index=5)
   &FN_InitializeAnimTrailBoneIndecies(AnimTrail Selection=1, First Bone or Socket Index=7, Second Bone or Socket Index=8) {module=FN_InitializeAnimTrailBoneIndecies001}
   /Game/Effects/Niagara/Modules/AnimTrail/FN_InitializeAnimTrail/FN_InitializeAnimTrail(AnimTrailWidth=2.75, Color=<rgba 0.0487 0.0605 0.0885 2>, Lifetime=0.1, Spawn Mode:ENiagaraSkeletalSpawning=::NewEnumerator0)
  ]
  particleUpdateScript [
   /Game/Effects/Niagara/Modules/AnimTrail/FN_UpdateAnimTrail/FN_UpdateAnimTrail()
   /Game/Effects/Niagara/Modules/Materials/Material_v2(PostMult=0.85)
   &ScaleColor(
    Scale Alpha=&FloatFromCurve(FloatCurve=DataInterface() {propertyValues {Curve {keys [
         {arriveTangent=-16.7667, interpMode=RCIM_Cubic, time=0, value=0}
         {arriveTangent=2.2222, interpMode=RCIM_Cubic, leaveTangent=2.2222, time=0.15, value=1}
         <at 0.45 1>
         {arriveTangent=-2.6227, interpMode=RCIM_Cubic, leaveTangent=-2.6227, tangentMode=RCTM_User, time=0.925, value=0}
        ]}}})
    Scale RGB=&VectorFromCurve(VectorCurve=DataInterface() {propertyValues {
       XCurve {keys [<at 0.15 1 RCIM_Cubic>,<at 1 0>]}
       YCurve {keys [<at 0.15 1 RCIM_Cubic>,<at 1 0>]}
       ZCurve {keys [<at 0.15 1 RCIM_Cubic>,<at 1 0>]}
      }})
    ScaleA=true
    ScaleRGB=true
   ) {module=ScaleColor001}
   &ScaleRibbonWidth(Ribbon Width Scale=&FloatFromCurve(FloatCurve=DataInterface() {propertyValues {Curve {keys [<at 0 1>,<at 0.4436 1.9009>,<at 1 2.5>]}}}))
  ]
  rendererClasses [&NiagaraRibbonRendererProperties]
  renderers [{rendererClass=&NiagaraRibbonRendererProperties}]
 }
 Lazy_Sparks {
  emitterUpdateScript [
   &EmitterState(
    Delay First Loop Only=true
    Inactive Response=::NewEnumerator1
    Life Cycle Mode=::NewEnumerator1
    Loop Delay=2
    Loop Duration=$System.LoopDuration
    Recalculate Duration Each Loop=true
    UseLoopDelay=true
   )
   &SpawnBurst_Instantaneous(
    Spawn Count=&Multiply_Int(A=&UniformRangedInt(Minimum=7, Maximum=12), B=$System.RandomSpawnBurst)
    Use Spawn Probability=false
   )
   &SpawnBurst_Instantaneous(
    Spawn Count=&Multiply_Int(A=&UniformRangedInt(Minimum=10, Maximum=14), B=$System.RandomSpawnBurst)
    Spawn Time=0.15
    Use Spawn Probability=false
   ) {module=SpawnBurst_Instantaneous001}
  ]
  particleSpawnScript [
   &InitializeParticle(
    Color=<rgba 4 4 4 1>
    Lifetime=&UniformRangedFloat(Minimum=0.1, Maximum=0.45)
    Module.Write Color=true
    Module.Write Lifetime=true
    Module.Write Mass=true
    Module.Write Material Random=true
    Module.Write Position=false
    Module.Write Scale=false
    Module.Write SpriteRotation=false
    Module.Write SpriteSize=true
    Position=&SimulationPosition()
    Sprite Size=&UniformRangedVector2D(Minimum=<1 1>, Maximum=<2 2>)
   )
   /SetVariables(Particles.SpawnGroup=$System.RandomSpawnGroup) {isSetParameters=true}
   &SphereLocation(
    Hemisphere X=true
    MaskSpawnGroup=false
    Module.Override Randomness=false
    Sphere Origin:NiagaraPosition=/Niagara/DynamicInputs/Helpers/SimulationPosition()
    Sphere Radius=20
    Surface Only=false
   )
   &SkeletalMeshLocation(
    Apply Sampled Orientation to Mesh Particles=true
    Bone / Socket Index=&Add_Integer(A=$Particles.SpawnGroup, B=15)
    Bone Sampling Mode=::NewEnumerator4
    Inherit Velocity=false
    Interpolate Mesh Sampling=false
    MaskSpawnGroup=false
    Skeletal Mesh=$User.SkeletalMesh
   )
   &AddVelocity(Velocity=&Multiply_Vector(
     A=$Output.SphereLocation.SphereNormal
     B=&VectorFromFloat(Value=&UniformRangedFloat(Minimum=30, Maximum=150))
    )) {module=AddVelocity001}
   &AddVelocity(Velocity=&RotateVectorByQuaternion(
     Vector To Rotate=&MaskVectorBySpawnGroup(
      Masked Vector=&UniformRangedVector(Minimum=<-250 -250 -200>, Maximum=<250 250 600>)
      Passthrough Vector=&UniformRangedVector(Minimum=<-250 -250 -200>, Maximum=<250 250 600>)
     )
     Quaternion=$Output.SkeletalMeshLocation.SampledRotation
    ))
   &CameraOffset(Camera Offset Amount=25)
   /Niagara/Modules/Audio/PlayAudio(
    Audio Player=DataInterface() {propertyValues {
      Attenuation=null
      Concurrency=null
      MaxPlaysPerTick=10
      SoundToPlay=/BRCosmetics/Sounds/Gliders/Hightower_Squash/Hightower_Squash_Strikes_Cue
      bAllowLoopingOneShotSounds=false
      bLimitPlaysPerTick=true
      bOnlyActiveDuringGameplay=false
      bStopWhenComponentIsDestroyed=true
      ConfigurationUserParameter {parameter {name=None, typeDefHandle {registeredTypeIndex=1}, varData []}}
      ParameterNames []
     }}
    PlayAudio=true
   )
  ]
  particleUpdateScript [
   &ParticleState()
   &AccelerationForce(Acceleration=<0 0 -4900>)
   &SolveForcesAndVelocity(Clamp Velocity=false, Limit Acceleration=false)
   &ScaleSpriteSizeBySpeed(Max Scale Factor=<1 3>)
   &ScaleSpriteSize(Scale Factor=&Vector2DFromFloat(Value=&FloatFromCurve(FloatCurve=DataInterface() {propertyValues {Curve {keys [{arriveTangent=-6.725, interpMode=RCIM_Cubic, leaveTangent=-6.725, tangentMode=RCTM_User, time=0.65, value=1}, <at 1 0 RCIM_Cubic>]}}})))
   &ScaleColor(
    Scale RGB=&MakeVectorFromLinearColorRGB(Color=&ColorFromCurve(ColorCurve=DataInterface() {propertyValues {
        AlphaCurve {keys [<at 0 1 RCIM_Cubic>,<at 1 1 RCIM_Cubic>]}
        BlueCurve {keys [<at 0.2 10 RCIM_Cubic>,<at 0.7 0.5 RCIM_Cubic>]}
        GreenCurve {keys [<at 0.2 9.9741 RCIM_Cubic>,<at 0.7 0.5 RCIM_Cubic>]}
        RedCurve {keys [<at 0.2 9.375 RCIM_Cubic>,<at 0.7 0.5 RCIM_Cubic>]}
       }}))
    ScaleA=true
    ScaleRGB=true
   )
  ]
  rendererClasses [&NiagaraSpriteRendererProperties]
  renderers [{rendererClass=&NiagaraSpriteRendererProperties}]
 }
 Lightning_Mesh {
  emitterUpdateScript [
   &EmitterState(
    Delay First Loop Only=true
    Inactive Response=::NewEnumerator1
    Life Cycle Mode=::NewEnumerator1
    Loop Delay=2
    Loop Duration=$System.LoopDuration
    Recalculate Duration Each Loop=true
    UseLoopDelay=true
   )
   &SpawnBurst_Instantaneous(Spawn Count=$System.RandomSpawnBurst, Use Spawn Probability=true)
   &SpawnBurst_Instantaneous(Spawn Count=$System.RandomSpawnBurst, Spawn Time=0.15, Use Spawn Probability=true) {module=SpawnBurst_Instantaneous001}
  ]
  particleSpawnScript [
   /SetVariables(Particles.SpawnGroup=$System.RandomSpawnGroup) {isSetParameters=true}
   /Game/Effects/Niagara/Modules/Math/FN_EvenNumberToSign(Input:NiagaraInt32=$Particles.SpawnGroup, Invert Sign=true, Output Float Attribute=true)
   &InitializeParticle(
    Color=<rgba 1.9271 1.9844 2 1>
    Lifetime=&UniformRangedFloat(Minimum=0.1, Maximum=0.15)
    Mesh Scale=&Multiply_VectorByFloat(Vector=&UniformRangedVector(Minimum=<15 6 6>, Maximum=<22.5 12 12>), Float=$Particles.Sign)
    Module.Write Color=true
    Module.Write Lifetime=true
    Module.Write Mass=true
    Module.Write Material Random=true
    Module.Write Position=false
    Module.Write Scale=true
    Module.Write SpriteRotation=false
    Module.Write SpriteSize=false
    Position=/Niagara/Functions/Localspace/SimulationPosition()
   )
   &DynamicMaterialParameters(
    Index 0 Param 1=6
    Index 0 Param 3=&UniformRangedFloat(Minimum=1.45, Maximum=2.2)
    Index 0 Param 4=&UniformRangedFloat(Minimum=-60, Maximum=60)
    Index 2 Param 1=&UniformRangedFloat(Minimum=0, Maximum=1)
    Index 2 Param 2=&UniformRangedFloat(Minimum=1, Maximum=4)
    Index 2 Param 3=1
    Module.Param0WriteEnabled=true
    Module.Param0WriteEnabled2=true
    Module.Param1WriteEnabled=false
    Module.Param1WriteEnabled2=true
    Module.Param2WriteEnabled=true
    Module.Param2WriteEnabled2=true
    Module.Param3WriteEnabled=true
    Module.Param3WriteEnabled2=false
    Write Parameter Index 2=true
   )
   /SetVariables(
    Particles.FlareCameraOffset=4
    Particles.FlareOffset:Vector3f=&UniformRangedVector(Minimum=<5 -1 -1>, Maximum=<100 1 1>)
    Particles.FlareOffsetSecondary:Vector3f=&UniformRangedVector(Minimum=<-5 -1 -1>, Maximum=<-100 1 1>)
    Particles.FlareSize:Vector2f=&Vector2DFromFloat(Value=&UniformRangedFloat(Minimum=55, Maximum=60))
    Particles.LensFlare_Color:LinearColor=&MultiplyLinearColor(
     Linear Color=$Particles.Color
     Scale Factor=&MakeVector4_FromLinearColor(LinearColor=$System.RandomFlareColor)
    )
    Particles.LightningMeshOffset:Vector3f=&UniformRangedVector(Minimum=<-2 -1 -1>, Maximum=<2 1 1>)
   ) {isSetParameters=true}
   &PlayAudio(
    Audio Player=DataInterface() {propertyValues {
      Attenuation=null
      Concurrency=null
      MaxPlaysPerTick=10
      SoundToPlay=&Hightower_Squash_Strikes_Cue
      bAllowLoopingOneShotSounds=false
      bLimitPlaysPerTick=true
      bOnlyActiveDuringGameplay=false
      bStopWhenComponentIsDestroyed=true
      ConfigurationUserParameter {parameter {name=None, typeDefHandle {registeredTypeIndex=1}, varData []}}
      ParameterNames []
     }}
    PlayAudio=true
   )
  ]
  particleUpdateScript [
   &ParticleState()
   &SkeletalMeshLocation(
    Apply Sampled Orientation to Mesh Particles=true
    Apply To Particle Position=false
    Bone / Socket Index=&Add_Integer(A=$Particles.Initial.SpawnGroup, B=15)
    Bone Sampling Mode=::NewEnumerator4
    Inherit Velocity=false
    Interpolate Mesh Sampling=false
    MaskSpawnGroup=false
    Skeletal Mesh=$User.SkeletalMesh
   )
   /BRCosmetics/Athena/Effects/Particles/Skydiving/NS_Contrail_AmmoBelt/Module/FN_QuaterionSwizzle(Quat:Quat4f=$Output.SkeletalMeshLocation.SampledRotation)
   &DynamicMaterialParameters(
    Index 0 Param 2=&FloatFromCurve(FloatCurve=DataInterface() {propertyValues {Curve {keys [{arriveTangent=-2.9283, interpMode=RCIM_Cubic, leaveTangent=-2.9283, tangentMode=RCTM_User, time=0.1, value=1}, <at 1 -1>]}}})
    Index 2 Param 4=&FloatFromCurve(FloatCurve=DataInterface() {propertyValues {Curve {keys [<at 0 1>,<at 0.2 0>,<at 0.4625 0.6907>,<at 1 1>]}}})
    Module.Param0WriteEnabled=false
    Module.Param0WriteEnabled2=false
    Module.Param1WriteEnabled=true
    Module.Param1WriteEnabled2=false
    Module.Param2WriteEnabled=false
    Module.Param2WriteEnabled2=false
    Module.Param3WriteEnabled=false
    Module.Param3WriteEnabled2=true
    Write Parameter Index 2=true
   ) {module=DynamicMaterialParameters001}
   &ScaleColor(
    Scale RGB=&MakeVectorFromLinearColorRGB(Color=&ColorFromCurve(ColorCurve=DataInterface() {propertyValues {
        AlphaCurve {keys []}
        BlueCurve {keys [<at 0 2 RCIM_Cubic>,<at 0.161 4>,<at 0.317 1>,<at 1 4 RCIM_Cubic>]}
        GreenCurve {keys [<at 0 1.8653 RCIM_Cubic>,<at 0.161 4>,<at 0.317 1>,<at 1 2.6922 RCIM_Cubic>]}
        RedCurve {keys [<at 0 1.5312 RCIM_Cubic>,<at 0.161 4>,<at 0.317 1>,<at 1 0.8958 RCIM_Cubic>]}
       }}))
    ScaleA=true
    ScaleRGB=true
   )
   /SetVariables(Particles.MeshOrientation=&MultiplyQuaternion(
     Quaternion A=&MakeQuatFromAxisAngle(Axis=$Engine.Owner.SystemXAxis, AngleInDegrees=50, NormalizeAxis=true)
     Quaternion B=&MultiplyQuaternion(
      Quaternion A=&MakeQuatFromAxisAngle(
       Axis=&RotateVectorByQuaternion(Vector To Rotate=<0 1 0>, Quaternion=$Particles.FN_QuaterionSwizzle.Quat)
       AngleInDegrees=/Game/Effects/Niagara/DynamicInputs/Math/FN_Fit01(Clamp:NiagaraBool=false, Value:NiagaraFloat=$Particles.MaterialRandom, DestMin:NiagaraFloat=-35, DestMax:NiagaraFloat=10)
      )
      Quaternion B=$Particles.FN_QuaterionSwizzle.Quat
     )
    )) {isSetParameters=true}
   /SetVariables(Particles.Position=&Add_Vector(
     A=$Output.SkeletalMeshLocation.SampledPosition
     B=&RotateVectorByQuaternion(Vector To Rotate=$Particles.LightningMeshOffset, Quaternion=$Output.SkeletalMeshLocation.SampledRotation)
    )) {isSetParameters=true}
   /SetVariables(
    Particles.FlashColor:LinearColor=&ColorFromCurve(ColorCurve=DataInterface() {propertyValues {
       AlphaCurve {keys [<at 0.4344 0>,<at 0.502 1>,<at 0.6946 1 RCIM_Cubic>,<at 0.8246 0 RCIM_Cubic>]}
       BlueCurve {keys [<at 0.664 2 RCIM_Cubic>,<at 1 2 RCIM_Cubic>]}
       GreenCurve {keys [<at 0.664 1.9924 RCIM_Cubic>,<at 1 1.9822 RCIM_Cubic>]}
       RedCurve {keys [<at 0.664 1.9688 RCIM_Cubic>,<at 1 1.9271 RCIM_Cubic>]}
      }})
    Particles.FlashPosition:NiagaraPosition=$Particles.Position
    Particles.FlashScale:Vector2f=&Multiply_Vector2(
     A=$Particles.Initial.FlareSize
     B=&Vector2DFromFloat(Value=&FloatFromCurve(FloatCurve=DataInterface() {propertyValues {Curve {keys [<at 0 1>, {arriveTangent=2.6204, interpMode=RCIM_Cubic, leaveTangent=2.6204, tangentMode=RCTM_User, time=0.1732, value=1.6257}, <at 1 0>]}}}))
    )
   ) {isSetParameters=true}
   /SetVariables(
    Particles.FlareStrike:Vector2f=&Multiply_Vector2(A=$Particles.FlashScale, B=<1.5 1.5>)
    Particles.LensFlare_Color:LinearColor=&MultiplyLinearColor(
     Linear Color=$Particles.Initial.LensFlare_Color
     Scale Factor=&MakeFloat4FromVectorAndFloat(
      "Vector (RGB)"=&VectorFromFloat(Value=&FloatFromCurve(FloatCurve=DataInterface() {propertyValues {Curve {keys [<at 0 1>,<at 0.1 1.5>,<at 0.3094 1.2707>,<at 1 1>]}}}))
      "Float (Alpha)"=&FloatFromCurve(FloatCurve=DataInterface() {propertyValues {Curve {keys [<at 0 0>,<at 0.1 1>,<at 0.75 1>,<at 1 0>]}}})
     )
    )
    Particles.LensFlare_Position:Vector3f=&Add_Vector(
     A=Linked() {linkedVariable {name=Particles.Position, type=NiagaraPosition}}
     B=&RotateVectorByQuaternion(
      Vector To Rotate=&MaskVectorBySpawnGroup(Masked Vector=$Particles.FlareOffset, Passthrough Vector=$Particles.FlareOffsetSecondary, Spawn Group Mask=$Particles.Initial.SpawnGroup)
      Quaternion=$Particles.MeshOrientation
     )
    )
   ) {isSetParameters=true}
   &FN_Light_v3(
    Color=&ColorFromCurve(ColorCurve=DataInterface() {propertyValues {
       AlphaCurve {keys []}
       BlueCurve {keys [<at 0 1 RCIM_Cubic>,<at 0.2084 5 RCIM_Cubic>,<at 0.5725 1 RCIM_Cubic>,<at 1 1 RCIM_Cubic>]}
       GreenCurve {keys [<at 0 0.9237 RCIM_Cubic>,<at 0.2084 4.469 RCIM_Cubic>,<at 0.5725 0.9237 RCIM_Cubic>,<at 1 0.9237 RCIM_Cubic>]}
       RedCurve {keys [<at 0 0.8 RCIM_Cubic>,<at 0.2084 3.6082 RCIM_Cubic>,<at 0.5725 0.8 RCIM_Cubic>,<at 1 0.8 RCIM_Cubic>]}
      }})
    Create Color Attribute=true
    Create Exponent Attribute=true
    Create Position Attribute=true
    Create Radius Attribute=true
    Enable Particle Light Enable=true
    Exponent=1
    Light Enabled=HlslExpression() {hlslExpression="Modulo (Particles.UniqueID, System.RandomSpawnBurst) == 1 ? 1.0: 0"}
    Position:Vector3f=&Add_Vector(
     A=Linked() {linkedVariable {name=Particles.Position, type=NiagaraPosition}}
     B=&RotateVectorByQuaternion(
      Vector To Rotate=&MaskVectorBySpawnGroup(Masked Vector=<25 0 0>, Passthrough Vector=<-25 0 0>, Spawn Group Mask=$Particles.Initial.InitalSpawnGroup)
      Quaternion=$Particles.MeshOrientation
     )
    )
    Radius=75
    Scale Color=5
   )
  ]
  rendererClasses [
   &NiagaraMeshRendererProperties
   &NiagaraLightRendererProperties
   &NiagaraSpriteRendererProperties
  ]
  renderers [
   {rendererClass=&NiagaraMeshRendererProperties, rendererIndex=0}
   {rendererClass=&NiagaraLightRendererProperties, rendererIndex=1}
   {rendererClass=&NiagaraSpriteRendererProperties, rendererIndex=2}
   {rendererClass=&NiagaraSpriteRendererProperties, rendererIndex=3}
   {rendererClass=&NiagaraSpriteRendererProperties, rendererIndex=4}
  ]
 }
 Lightning_Strike {
  emitterUpdateScript [
   &EmitterState(
    Inactive Response=::NewEnumerator1
    Life Cycle Mode=::NewEnumerator1
    Loop Duration=&UniformRangedFloat(Minimum=1.5, Maximum=6)
    Recalculate Duration Each Loop=true
    UseLoopDelay=false
   )
   &SpawnBurst_Instantaneous(
    Spawn Count=&MakeCustomIntFromBool(Bool=$User.bIsFrontEndPreview, True INT=0, False INT=&UniformRangedInt(Minimum=0, Maximum=2))
    Use Spawn Probability=false
   )
   &SpawnBurst_Instantaneous(
    Spawn Count=&MakeCustomIntFromBool(Bool=$User.bIsFrontEndPreview, True INT=0, False INT=&UniformRangedInt(Minimum=0, Maximum=1))
    Spawn Time=&UniformRangedFloat(Minimum=0.35, Maximum=0.85)
    Use Spawn Probability=false
   ) {module=SpawnBurst_Instantaneous001}
  ]
  particleSpawnScript [
   &InitializeParticle(
    Lifetime=&UniformRangedFloat(Minimum=0.2, Maximum=0.25)
    Mesh Scale=&Multiply_Vector(A=&UniformRangedVector(Minimum=<3 0.75 0.75>, Maximum=<4 1.25 1.25>), B=<1 1 1>)
    Module.Write Color=true
    Module.Write Lifetime=true
    Module.Write Mass=false
    Module.Write Material Random=true
    Module.Write Position=false
    Module.Write Scale=true
    Module.Write SpriteRotation=false
    Module.Write SpriteSize=false
    Position=&SimulationPosition()
   )
   &DynamicMaterialParameters(
    Index 0 Param 4=&UniformRangedFloat(Minimum=0, Maximum=1)
    Module.Param0WriteEnabled=false
    Module.Param1WriteEnabled=false
    Module.Param2WriteEnabled=false
    Module.Param3WriteEnabled=true
   )
   &InitialMeshOrientation(
    Orientation Vector=<0 0 1>
    Rotation=&UniformRangedVector(Minimum=<-1 0 0>, Maximum=<1 0 0>)
    Use Orientation Vector=true
    Use Rotation Vector=true
   )
   &CylinderLocation(
    Coordinate Space:ENiagaraCoordinateSpace=::Simulation
    Cylinder Height=20
    Cylinder Origin:Vector3f=&Add_Vector(
     A=Linked() {linkedVariable {name=Engine.Owner.Position, type=NiagaraPosition}}
     B=&RotateVectorByQuaternion(Vector To Rotate=<0 100 -800>, Quaternion=$Engine.Owner.Rotation)
    )
    Cylinder Radius=500
    MaskSpawnGroup=false
    Module.Enable Lathe Profile=false
    Module.Override Randomness=false
    Module.Surface Only=true
    Surface Only Band Thickness=100
    Use Endcaps In Surface Only Mode=false
   )
   &PlayAudio(
    Audio Player=DataInterface() {propertyValues {
      Attenuation=null
      Concurrency=null
      MaxPlaysPerTick=10
      SoundToPlay=&Hightower_Squash_Strikes_Cue
      bAllowLoopingOneShotSounds=false
      bLimitPlaysPerTick=true
      bOnlyActiveDuringGameplay=false
      bStopWhenComponentIsDestroyed=true
      ConfigurationUserParameter {parameter {name=None, typeDefHandle {registeredTypeIndex=1}, varData []}}
      ParameterNames []
     }}
    PlayAudio=true
   )
  ]
  particleUpdateScript [
   &ParticleState()
   &ScaleColor(
    Scale RGB=&MakeVectorFromLinearColorRGB(Color=&ColorFromCurve(ColorCurve=DataInterface() {propertyValues {
        AlphaCurve {keys []}
        BlueCurve {keys [<at 0 2 RCIM_Cubic>,<at 1 1 RCIM_Cubic>]}
        GreenCurve {keys [<at 0 1.9431 RCIM_Cubic>,<at 1 0.8596 RCIM_Cubic>]}
        RedCurve {keys [<at 0 1.8021 RCIM_Cubic>,<at 1 0.6667 RCIM_Cubic>]}
       }}))
    ScaleA=true
    ScaleRGB=true
   )
   &DynamicMaterialParameters(
    Index 0 Param 1=&FloatFromCurve(FloatCurve=DataInterface() {propertyValues {Curve {keys [
         <at 0 1 RCIM_Cubic>
         {arriveTangent=-1.2489, interpMode=RCIM_Cubic, leaveTangent=-1.2489, time=0.3515, value=0.6647}
         {arriveTangent=-1.8243, interpMode=RCIM_Cubic, leaveTangent=-1.8243, time=0.3565, value=0.5547}
         {arriveTangent=-2.8364, interpMode=RCIM_Cubic, leaveTangent=-2.8364, tangentMode=RCTM_User, time=0.6784, value=0.0685}
        ]}}})
    Index 0 Param 2=&FloatFromCurve(FloatCurve=DataInterface() {propertyValues {Curve {keys [
         {arriveTangent=3.0384, interpMode=RCIM_Cubic, leaveTangent=3.0384, tangentMode=RCTM_User, time=0, value=0.25}
         {arriveTangent=0.625, interpMode=RCIM_Cubic, leaveTangent=0.625, time=0.5, value=0.5}
         {arriveTangent=-1.25, interpMode=RCIM_Cubic, leaveTangent=-1.25, time=0.8, value=0.75}
         <at 0.9 0>
        ]}}})
    Index 0 Param 3=&FloatFromCurve(FloatCurve=DataInterface() {propertyValues {Curve {keys [
         {arriveTangent=0.8047, interpMode=RCIM_Cubic, leaveTangent=0.8047, tangentMode=RCTM_User, time=0, value=0.25}
         {arriveTangent=1.25, interpMode=RCIM_Cubic, leaveTangent=1.25, time=0.55, value=0.5}
         {arriveTangent=-2, interpMode=RCIM_Cubic, leaveTangent=-2, time=0.6, value=1}
         <at 0.8 0>
        ]}}})
    Module.Param0WriteEnabled=true
    Module.Param1WriteEnabled=true
    Module.Param2WriteEnabled=true
    Module.Param3WriteEnabled=false
   ) {module=DynamicMaterialParameters001}
  ]
  rendererClasses [&NiagaraMeshRendererProperties]
  renderers [{rendererClass=&NiagaraMeshRendererProperties, rendererIndex=0}]
 }
 Rotating_Air_Meshes {
  emitterUpdateScript [
   &EmitterState()
   &SpawnRate(SpawnRate=&UniformRangedFloat(Minimum=6, Maximum=14), Use Spawn Probability=false)
  ]
  particleSpawnScript [
   &FN_Orbit_Init_Variables_v2()
   &InitializeParticle(
    Color=<rgba 0.0546 0.0662 0.0677 0.35>
    Lifetime=&UniformRangedFloat(Minimum=1, Maximum=1.5)
    Mesh Scale=&UniformRangedVector(Minimum=<5 5 8>, Maximum=<8 8 11>)
    Module.Write Color=true
    Module.Write Lifetime=true
    Module.Write Mass=false
    Module.Write Material Random=true
    Module.Write Position=false
    Module.Write Scale=true
    Module.Write SpriteRotation=false
    Module.Write SpriteSize=false
    Position=&SimulationPosition()
   )
   &InitialMeshOrientation(
    Orientation Vector=<0 0 1>
    Rotation=&UniformRangedVector(Minimum=<0 0 -1>, Maximum=<0 0 1>)
    Use Orientation Vector=true
    Use Rotation Vector=true
   )
   &DynamicMaterialParameters(
    Index 0 Param 1=&UniformRangedFloat(Minimum=0.1, Maximum=0.25)
    Index 0 Param 2=1
    Module.Param0WriteEnabled=true
    Module.Param1WriteEnabled=true
    Module.Param2WriteEnabled=true
    Module.Param3WriteEnabled=true
   )
   &FN_MeshRenderer_v2(
    Color:LinearColor=&UniformRangedLinearColor(Minimum=<rgba 0.8229 0.8229 0.8229 1>, Maximum=<rgba 0.8802 0.8802 0.8802 1>)
    Create COlor Attribute=true
    Create Dynamic Param Attribute=true
    Create Scale Attribute=true
    DynamicMaterial:Vector4f=&MakeVector4(
     X=&UniformRangedFloat(Minimum=0, Maximum=1)
     Y=&UniformRangedFloat(Minimum=5, Maximum=6)
     Z=0
     W=2
    )
    Scale:Vector3f=&UniformRangedVector(Minimum=<0.85 0.85 0.5>, Maximum=<1 1 1.5>)
   )
  ]
  particleUpdateScript [
   &SkeletalMeshLocation(
    Apply Sampled Orientation to Mesh Particles=true
    Apply To Particle Position=false
    Bone / Socket Index=1
    Bone Sampling Mode=::NewEnumerator4
    Inherit Velocity=false
    Interpolate Mesh Sampling=false
    MaskSpawnGroup=false
    Skeletal Mesh=$User.SkeletalMesh
   )
   &ParticleState()
   &ScaleMeshSize(Scale Factor=&VectorFromCurve(VectorCurve=DataInterface() {propertyValues {
       XCurve {keys [<at 0 1 RCIM_Cubic>,<at 1 0.65>]}
       YCurve {keys [<at 0 1 RCIM_Cubic>,<at 1 0.65>]}
       ZCurve {keys [<at 0 1.5 RCIM_Cubic>,<at 1 1>]}
      }}))
   &ScaleColor(
    Scale Alpha=&FloatFromCurve(FloatCurve=DataInterface() {propertyValues {Curve {keys [
         <at 0 0 RCIM_Cubic>
         <at 0.3 1 RCIM_Cubic>
         {arriveTangent=-4.5166, interpMode=RCIM_Cubic, leaveTangent=-4.5166, tangentMode=RCTM_User, time=1, value=0}
        ]}}})
    ScaleA=true
    ScaleRGB=false
   )
   /SetVariables(Particles.MeshOrientation=&MultiplyQuaternion(
     Quaternion A=&MakeQuatFromAxisAngle(
      Axis=&RotateVectorByQuaternion(Vector To Rotate=<1 0 0>, Quaternion=$Output.SkeletalMeshLocation.SampledRotation)
      AngleInDegrees=&Add_Float(
       A=&Multiply_Float(A=$Emitter.NormalizedLoopAge, B=360)
       B=&Multiply_Float(A=$System.FN_Integrator_v2.Attr, B=360)
      )
     )
     Quaternion B=&MultiplyQuaternion(
      Quaternion A=$Output.SkeletalMeshLocation.SampledRotation
      Quaternion B=&MultiplyQuaternion(Quaternion A=$Particles.Orientation, Quaternion B=$Particles.Initial.MeshOrientation)
     )
    )) {isSetParameters=true}
   /SetVariables(Particles.Position=&Add_Vector(
     A=$Output.SkeletalMeshLocation.SampledPosition
     B=&RotateVectorByQuaternion(
      Vector To Rotate=&Add_Vector(
       A=<-10 0 8>
       B=&Multiply_VectorByFloat(Vector=<-75 0 0>, Float=&Multiply_Float(A=$Particles.NormalizedAge, B=1.5))
      )
      Quaternion=$Output.SkeletalMeshLocation.SampledRotation
     )
    )) {isSetParameters=true}
   &FN_MeshRenderer_v2(
    Color:LinearColor=&MultiplyLinearColor(
     Linear Color=$Particles.Initial.FN_MeshRenderer_v2.Color
     Scale Factor=&MakeFloat4FromVectorAndFloat(
      "Vector (RGB)"=<1 1 1>
      "Float (Alpha)"=&FloatFromCurve(FloatCurve=DataInterface() {propertyValues {Curve {keys [
           <at 0 0 RCIM_Cubic>
           {arriveTangent=1.3333, leaveTangent=1.3333, time=0.15, value=1}
           <at 0.85 1 RCIM_Cubic>
           {arriveTangent=-4.5166, interpMode=RCIM_Cubic, leaveTangent=-4.5166, tangentMode=RCTM_User, time=1, value=0}
          ]}}})
     )
    )
    Create COlor Attribute=true
    Create Position Attribute=true
    Create Scale Attribute=true
    Scale:Vector3f=&Multiply_Vector(
     A=$Particles.Initial.FN_MeshRenderer_v2.Scale
     B=&VectorFromCurve(VectorCurve=DataInterface() {propertyValues {
        XCurve {keys [<at 0 1.25 RCIM_Cubic>,<at 1 0.75>]}
        YCurve {keys [<at 0 1.25 RCIM_Cubic>,<at 1 0.75>]}
        ZCurve {keys [<at 0 1.5 RCIM_Cubic>,<at 1 1>]}
       }})
    )
   ) {module=FN_MeshRenderer_v2001}
  ]
  rendererClasses [&NiagaraMeshRendererProperties]
  renderers [
   {rendererClass=&NiagaraMeshRendererProperties, rendererIndex=0}
   {rendererClass=&NiagaraMeshRendererProperties, rendererIndex=1}
  ]
 }
}
summary {
 systemName=NS_Glider_HighTower_Squash
 emitters [
  {emitterName=Rotating_Air_Meshes, rendererClasses [&NiagaraMeshRendererProperties]}
  {
   emitterName=Electric_Arcs_Hands
   rendererClasses [&NiagaraSpriteRendererProperties, &NiagaraMeshRendererProperties]
  }
  {emitterName=AirGlow, rendererClasses [&NiagaraSpriteRendererProperties, &NiagaraLightRendererProperties]}
  {
   emitterName=Lightning_Mesh
   rendererClasses [
    &NiagaraMeshRendererProperties
    &NiagaraLightRendererProperties
    &NiagaraSpriteRendererProperties
   ]
  }
  {emitterName=Debris, rendererClasses [&NiagaraMeshRendererProperties]}
  {emitterName=FN_AnimTrail_V4, rendererClasses [&NiagaraRibbonRendererProperties]}
  {emitterName=EnergyFlow, rendererClasses [&NiagaraSpriteRendererProperties]}
  {emitterName=ElectricMeshArcsArms, rendererClasses [&NiagaraMeshRendererProperties]}
  {emitterName=Clouds, rendererClasses [&NiagaraSpriteRendererProperties]}
  {emitterName=Lazy_Sparks, rendererClasses [&NiagaraSpriteRendererProperties]}
  {emitterName=Lightning_Strike, rendererClasses [&NiagaraMeshRendererProperties]}
 ]
 userVariables [
  User.bIsFrontEndPreview {type=NiagaraBool}
  User.bIsFullyDeployed {type=NiagaraBool, defaultValue=true}
  User.ForwardDot {type=NiagaraFloat, defaultValue=0, description=""}
  User.RibbonIDOffset {type=NiagaraInt32}
  User.RightDot {type=NiagaraFloat, defaultValue=0, description=""}
  User.SkeletalMesh {
   type=NiagaraDataInterfaceSkeletalMesh
   defaultValue {
    struct=NiagaraExt_VariableValue_DataInterface
    value {dataInterface=&self:NiagaraDataInterfaceSkeletalMesh_0}
   }
  }
 ]
}
